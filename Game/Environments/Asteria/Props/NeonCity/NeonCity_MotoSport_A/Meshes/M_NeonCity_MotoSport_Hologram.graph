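# Hologram material for the NeonCity MotoSport prop, faking a glitchy, see-through spinning projection on a lit surface.
# World position offset spins the mesh about a pivot, with matching rotated normals.
# Diffuse gets time-driven chromatic aberration, pixelation from a curve atlas offset by actor position, and a height-blended tint.
# Scanlines, fresnel, view-based skylight cubemap with night boost and sun passthrough shape color and emissive; dithered pixel depth offset softens depth.
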Material /Game/Environments/Asteria/Props/NeonCity/NeonCity_MotoSport_A/Meshes/M_NeonCity_MotoSport_Hologram
graph {
 properties {bUseMaterialAttributes=true, pixelDepthOffsetMode=PDOM_Legacy, usedShadingModels=Default Lit}
 nodes [
  NamedRerouteDeclaration(None=MaterialFunctionCall!(Scanline Amount) {materialFunction=/Game/Environments/Asteria/Sets/NeonCity/Techart/Holograms/NewHologramMaterial/MFScanlines}.Result) {name=Scanlines, nodeColor=<rgba 1 0 0.1641 1>}
  n1 = VectorParameter() {parameterName=Axis, defaultValue=<rgba 0 0 1 1>}
  n2 = Multiply(A=Time(), B=ScalarParameter() {defaultValue=0.1, parameterName=SpeedRot})
  n3 = TextureSampleParameter2D() {
   parameterName=Normal
   samplerType=SAMPLERTYPE_Normal
   texture=/Game/Characters/Player/Female/Medium/Bodies/F_MED_Masako/Textures/T_F_MED_Masako_N
  }
  n4 = MaterialFunctionCall(
   Rotation Axis=@n1.RGB
   Rotation Angle=Add(A=@n2, B=VectorParameter() {parameterName=SlightTilt, defaultValue=<rgba 0 0 0 1>}.RGB)
   World Space Vertex Normals=Transform(None=@n3.RGB)
  ) {materialFunction=&FixRotateAboutAxisNormals}
  NamedRerouteDeclaration(None=@n4.New Tangent Space Vertex Normal) {name=Normal, nodeColor=<rgba 0.0351 0.2693 1 1>}
  n5 = MaterialFunctionCall(Float3=OneMinus(None=CameraVectorWS())) {materialFunction=&BreakOutFloat3Components}
  n6 = ScalarParameter() {defaultValue=-1, parameterName=Offset SeeThrough}
  n7 = SkyLightEnvMapSample(
   Direction=MaterialFunctionCall(X=@n5.R, Y=@n5.R, Z=Add(A=@n5.B, B=@n6)) {materialFunction=&MakeFloat3}.Result
   Roughness=ScalarParameter() {defaultValue=0.1, parameterName=HologramRoughness}
  )
  n8 = ScalarParameter() {defaultValue=1, parameterName=CubemapBleed}
  n9 = Constant() {r=1}
  n10 = ScalarParameter() {defaultValue=5, parameterName=Night Boost}
  NamedRerouteDeclaration(None=Multiply(
    A=Multiply(A=@n7, B=@n8)
    B=MaterialFunctionCall(Value_Morning=@n9, Value_Day=@n9, Value_Evening=@n9, Value_Night=@n10) {materialFunction=&Scalar_Time_of_Day}.Result
   )) {name=Cubemap, nodeColor=<rgba 0.8984 0 1 1>}
  n11 = ActorPositionWS()
  n12 = Add(
   A=Divide(A=MaterialFunctionCall(Float3=@n11) {materialFunction=&BreakOutFloat3Components}.G) {constB=20.111}
   B=Time()
  )
  n13 = ScalarParameter() {defaultValue=0.33, parameterName=Glitch SPeed}
  n14 = CurveAtlasRowParameter(CurveTime=Frac(None=Multiply(A=@n12, B=@n13))) {
   atlas=/Game/Environments/Asteria/Sets/NeonCity/Techart/Holograms/NewHologramMaterial/HologramCurves
   curve=/Game/Environments/Asteria/Sets/NeonCity/Techart/Holograms/NewHologramMaterial/HologramCurve
   parameterName=Glitch
  }
  NamedRerouteDeclaration(None=@n14.B) {name=Glitch Blue, nodeColor=<rgba 1 0.5625 0 1>}
  n15 = Multiply(A=@n12, B=ScalarParameter() {defaultValue=0.33, parameterName=Glitch SPeed})
  n16 = LinearInterpolate(B=ScalarParameter() {defaultValue=0.003, parameterName=ChromaticAb}, Alpha=Abs(None=Sine(None=@n15) {period=20}))
  n17 = Constant()
  n18 = MaterialFunctionCall(R=@n16, G=@n17, B=@n17) {materialFunction=&AppendMany}
  n19 = TextureCoordinate()
  n20 = TextureObjectParameter() {parameterName=Diffuse, texture=/Game/Characters/Player/Female/Medium/Bodies/F_MED_Masako/Textures/T_F_MED_Masako_D}
  n21 = MaterialFunctionCall(R=@n17, G=@n16, B=@n17) {materialFunction=&AppendMany}
  n22 = MaterialFunctionCall(R=@n17, G=@n17, B=OneMinus(None=@n16)) {materialFunction=&AppendMany}
  n23 = MaterialFunctionCall(
   R=TextureSample(UVs=Add(A=@n19, B=@n18.RG), Tex=@n20) {texture=&T_F_MED_Masako_D}.R
   G=TextureSample(UVs=Add(A=@n19, B=@n21.RG), Tex=@n20) {texture=&T_F_MED_Masako_D}.G
   B=TextureSample(UVs=Add(A=@n19, B=@n22.RG), Tex=@n20) {texture=&T_F_MED_Masako_D}.B
  ) {materialFunction=&AppendMany}
  n24 = ScalarParameter() {defaultValue=30, parameterName=Tile}
  n25 = Multiply(A=Frac(None=@n15)) {constB=10}
  n26 = LinearInterpolate(A=Multiply(A=@n24) {constB=0.8}, B=@n24, Alpha=Divide(A=Round(None=@n25)) {constB=10})
  n27 = Multiply(A=@n19, B=@n26) {constB=5}
  n28 = TextureSample(UVs=Divide(A=Floor(None=@n27), B=@n26), Tex=@n20) {texture=&T_F_MED_Masako_D}
  n29 = Round(None=NamedRerouteUsage())
  n30 = ScalarParameter() {defaultValue=0.8, parameterName=Hologram Tint Min}
  n31 = ScalarParameter() {defaultValue=0.4, parameterName=Hologram Tint Max}
  n32 = Transform(None=WorldPosition().XYZ) {transformSourceType=TRANSFORMSOURCE_World, transformType=TRANSFORM_Local}
  n33 = VertexInterpolator(VS=PreSkinnedLocalBounds().Max)
  n34 = MaterialFunctionCall(Float3=Saturate(None=Divide(A=@n32, B=@n33.PS))) {materialFunction=&BreakOutFloat3Components}
  NamedRerouteDeclaration(None=LinearInterpolate(
    A=LinearInterpolate(A=@n23.RGB, B=@n28.RGB, Alpha=@n29)
    B=VectorParameter() {parameterName=Overall Hologram Tint, defaultValue=<rgba 1 0.2604 0.2879 1>}.RGB
    Alpha=LinearInterpolate(A=@n30, B=@n31, Alpha=@n34.B)
   )) {name=Color, nodeColor=<rgba 0.3125 0 1 1>}
  NamedRerouteUsage()
  n35 = TextureSampleParameter2D() {
   parameterName=Specular
   samplerType=SAMPLERTYPE_Masks
   texture=/Game/Characters/Player/Female/Medium/Bodies/F_MED_Masako/Textures/T_F_MED_Masako_S
  }
  n36 = MaterialFunctionCall() {materialFunction=&ObjectPivotPoint}
  n37 = VectorParameter() {parameterName=PivotOffset, defaultValue=<rgba 0 0 0 1>}
  n38 = ScalarParameter() {defaultValue=33, parameterName=PDO_Distance}
  n39 = ScalarParameter() {defaultValue=1, parameterName=PDoDistanceMult}
  n40 = NamedRerouteUsage()
  n41 = ScalarParameter() {defaultValue=1.5, parameterName=outy}
  n42 = ScalarParameter() {defaultValue=0.2, parameterName=inny}
  n43 = Transform(None=NamedRerouteUsage())
  n44 = Multiply(
   A=NamedRerouteUsage()
   B=LinearInterpolate(A=Multiply(A=@n40, B=@n41), B=Multiply(A=@n40, B=@n42), Alpha=Fresnel(Normal=@n43) {baseReflectFraction=0, exponent=3})
  )
  n45 = Add(
   A=LinearInterpolate(A=NamedRerouteUsage(), B=@n28.RGB, Alpha=ScalarParameter() {defaultValue=0.35, parameterName=Color/CubemapInfluence})
   B=@n44
  )
  n46 = Fresnel(ExponentIn=ScalarParameter() {defaultValue=2, parameterName=SeethroughSides}, Normal=NamedRerouteUsage()) {baseReflectFraction=0}
  n47 = ScalarParameter() {defaultValue=0.02, parameterName=Sunlight Passthrough}
  n48 = CollectionParameter() {collection=&FortniteMaterialParameters, parameterName=SunLightColor}
  n49 = Multiply(A=ComponentMask(None=Saturate(None=@n48)) {b=true, g=true, r=true}, B=NamedRerouteUsage())
  n50 = MaterialFunctionCall() {materialFunction=/Game/Packages/Fortress_Sky/Fortnite_SunScreenSpaceFalloff}
  n51 = Add(A=Clamp(None=LinearInterpolate(A=@n44, B=@n45, Alpha=@n46)), B=Multiply(A=Multiply(A=@n47, B=@n49), B=Clamp(None=@n50.Dot*-1)))
  n52 = MaterialFunctionCall!(
   Emissive=Multiply(A=@n51, B=ScalarParameter() {defaultValue=0.6, parameterName=EmissiveNess})
   Alpha=ScalarParameter() {defaultValue=1, group=Cosmetic Emissive EyeAdaptation, parameterName=Cosmetic Emissive EyeAdaptation Alpha}
   Material
  ) {materialFunction=/Game/Packages/Fortress_SharedMaterials/Base_Material_Functions/BaseCharacterShaderFunctions/MF_CosmeticEmissiveEyeAdaptation}
  SetMaterialAttributes!(
   MaterialAttributes
   Base Color=Multiply(A=@n51, B=ScalarParameter() {defaultValue=0.6, parameterName=ColorNess})
   Metallic=@n35.G
   Specular=@n35.G
   Roughness=@n35.G
   Anisotropy
   Emissive Color=@n52.Emissive
   Opacity
   Opacity Mask
   Normal=@n4.New Tangent Space Vertex Normal
   Tangent
   World Position Offset=RotateAboutAxis(NormalizedRotationAxis=@n1.RGB, RotationAngle=@n2, PivotPoint=Add(A=@n36.Object Pivot Location, B=@n37.RGB), Position=WorldPosition().XYZ)
   Subsurface Color
   Pixel Depth Offset=Multiply(
    A=Multiply(A=@n38, B=@n39)
    B=MaterialFunctionCall() {materialFunction=&DitherTemporalAA}.Result
   )
   Custom Data 1
  ) {attributeSetTypes [
    #1
    #2
    #3
    #4
    #5
    #6
    #7
    #8
    #9
    #10
    #11
    #12
    #13
    #14
   ]}
 ]
}
